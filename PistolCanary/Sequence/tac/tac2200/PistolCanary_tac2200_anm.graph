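# Animation level sequence for shot tac2200 of PistolCanary, animating the TentacleBoss at a 60 fps display rate.
# It possesses two TentacleBoss bindings, labeled for Hermes_Terrain and Test Level, each with a skeletal animation track.
# The Hermes_Terrain binding has one continuous animation section, while the Test Level binding has two separated sections and a visibility track.
# The sequence has no root tracks, such as camera cuts or audio.
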LevelSequence /PistolCanary/Sequence/tac/tac2200/PistolCanary_tac2200_anm
bindings [
 TentacleBoss - Hermes_Terrain {
  id=#1
  possessed_class=PistolCanary_TentacleBoss_C
  tracks [{type=MovieSceneSkeletalAnimationTrack, sections [{type=MovieSceneSkeletalAnimationSection, end_frame=4509, start_frame=3600}]}]
 }
 TentacleBoss - Test Level {
  id=#2
  possessed_class=PistolCanary_TentacleBoss_C
  tracks [
   {
    type=MovieSceneSkeletalAnimationTrack
    sections [
     {type=MovieSceneSkeletalAnimationSection, end_frame=4004, start_frame=3600}
     {type=MovieSceneSkeletalAnimationSection, end_frame=4509, start_frame=4105}
    ]
   }
   {
    type=MovieSceneVisibilityTrack
    sections [{type=MovieSceneVisibilitySection, channels [None {type=MovieSceneScriptingBoolChannel, default=true, const [true,1]}]}]
   }
  ]
 }
]
display_rate {denominator=1, numerator=60}
tick_resolution {denominator=1, numerator=24000}
tracks []
