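# PistolCanary_tac2200_aud is the audio Level Sequence for the tac2200 sequence, which involves a tentacle boss.
# Two master audio tracks, TentacleAudio and "Tentacle Rawrrr", play equal-length, staggered, overlapping sections at default volume and pitch.
# An event track on the possessed PistolCanary_AudioManager fires three trigger keys around those sections.
# The TentacleBoss actor (BGA_Tentacle) is bound but has no tracks of its own.
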LevelSequence /PistolCanary/Sequence/tac/tac2200/PistolCanary_tac2200_aud
bindings [
 PistolCanary_AudioManager {
  id=#1
  possessed_class=PistolCanary_AudioManager_C
  tracks [{
    type=MovieSceneEventTrack
    sections [{
      type=MovieSceneEventTriggerSection
      channels [None {
        type=MovieSceneScriptingEventChannel
        keys [
         [3610, "<Struct 'MovieSceneEvent' (0x852ff39ee20) {}>"]
         [3727, "<Struct 'MovieSceneEvent' (0x852ff39eec0) {}>"]
         [4508, "<Struct 'MovieSceneEvent' (0x852ff39ef60) {}>"]
        ]
       }]
     }]
   }]
 }
 TentacleBoss {id=#2, possessed_class=BGA_Tentacle_C, tracks []}
]
display_rate {denominator=1, numerator=60}
tick_resolution {denominator=1, numerator=24000}
tracks [
 {
  type=MovieSceneAudioTrack
  display_name=Tentacle Rawrrr
  sections [{
    type=MovieSceneAudioSection
    end_frame=5680
    start_frame=4365
    channels [
     Volume {type=MovieSceneScriptingFloatChannel, default=1}
     Pitch {type=MovieSceneScriptingFloatChannel, default=1}
     AttachActor {type=MovieSceneScriptingActorReferenceChannel}
    ]
   }]
 }
 {
  type=MovieSceneAudioTrack
  display_name=TentacleAudio
  sections [{
    type=MovieSceneAudioSection
    end_frame=4915
    start_frame=3600
    channels [
     Volume {type=MovieSceneScriptingFloatChannel, default=1}
     Pitch {type=MovieSceneScriptingFloatChannel, default=1}
     AttachActor {type=MovieSceneScriptingActorReferenceChannel}
    ]
   }]
 }
]
