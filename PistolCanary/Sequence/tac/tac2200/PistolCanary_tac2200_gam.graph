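# Gameplay Level Sequence for PistolCanary tac2200 that times events on the tentacle boss and a laser origin actor.
# Separate bindings cover the Hermes_Terrain boss, with two timed event triggers, and a test-level BGA_Tentacle, with its own event track.
# A spawn track keeps PistolCanary_BGA_LaserOrigin alive for a window that encloses its two event triggers and both terrain boss events.
# The laser origin uses an attach track with an identity transform; event payloads are empty here.
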LevelSequence /PistolCanary/Sequence/tac/tac2200/PistolCanary_tac2200_gam
bindings [
 TentacleBoss - Hermes_Terrain {
  id=#1
  possessed_class=PistolCanary_TentacleBoss_C
  tracks [{
    type=MovieSceneEventTrack
    sections [{
      type=MovieSceneEventTriggerSection
      channels [None {
        type=MovieSceneScriptingEventChannel
        keys [[3732, "<Struct 'MovieSceneEvent' (0x852ff39f3c0) {}>"], [4039, "<Struct 'MovieSceneEvent' (0x852ff39f460) {}>"]]
       }]
     }]
   }]
 }
 TentacleBoss - Test Level {
  id=#2
  possessed_class=BGA_Tentacle_C
  tracks [{
    type=MovieSceneEventTrack
    sections [{
      type=MovieSceneEventTriggerSection
      channels [None {type=MovieSceneScriptingEventChannel, const ["<Struct 'MovieSceneEvent' (0x852ff39eec0) {}>", 1]}]
     }]
   }]
 }
 PistolCanary_BGA_LaserOrigin {
  id=#3
  possessed_class=PistolCanary_BGA_LaserOrigin_C
  tracks [
   {
    type=MovieSceneSpawnTrack
    sections [{
      type=MovieSceneSpawnSection
      channels [None {type=MovieSceneScriptingBoolChannel, default=false, keys [[3600,false],[3725,true],[4340,false]]}]
     }]
   }
   {
    type=MovieScene3DTransformTrack
    sections [{
      type=MovieScene3DTransformSection
      channels [
       Location.X {type=MovieSceneScriptingDoubleChannel, default=0}
       Location.Y {type=MovieSceneScriptingDoubleChannel, default=0}
       Location.Z {type=MovieSceneScriptingDoubleChannel, default=0}
       Rotation.X {type=MovieSceneScriptingDoubleChannel, default=0}
       Rotation.Y {type=MovieSceneScriptingDoubleChannel, default=0}
       Rotation.Z {type=MovieSceneScriptingDoubleChannel, default=0}
       Scale.X {type=MovieSceneScriptingDoubleChannel, default=1}
       Scale.Y {type=MovieSceneScriptingDoubleChannel, default=1}
       Scale.Z {type=MovieSceneScriptingDoubleChannel, default=1}
      ]
     }]
   }
   {type=MovieScene3DAttachTrack, sections [{type=MovieScene3DAttachSection, end_frame=4340, start_frame=3600}]}
   {
    type=MovieSceneEventTrack
    sections [{
      type=MovieSceneEventTriggerSection
      channels [None {
        type=MovieSceneScriptingEventChannel
        keys [[3730, "<Struct 'MovieSceneEvent' (0x852ff39f960) {}>"], [4315, "<Struct 'MovieSceneEvent' (0x852ff39f8c0) {}>"]]
       }]
     }]
   }
  ]
 }
]
display_rate {denominator=1, numerator=60}
tick_resolution {denominator=1, numerator=24000}
tracks []
