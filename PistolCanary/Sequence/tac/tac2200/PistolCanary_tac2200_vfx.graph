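# VFX Level Sequence for shot tac2200 in PistolCanary: the tentacle boss charges and fires a sweeping laser.
# An attached charge-up Niagara system ramps its parameters as the boss's four body materials pulse and glow bluer.
# The attached beam then activates, swept by keyed rotation with a hand-keyed ImpactOffset, before BeamWidth collapses.
# A post-process volume blends around the attack, while a static CineCamera adds a brief shake and intensity spike after firing starts.
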LevelSequence /PistolCanary/Sequence/tac/tac2200/PistolCanary_tac2200_vfx
bindings [
 NiagaraComponent0 {
  id=#1
  parent_id=#2
  possessed_class=NiagaraComponent
  tracks [
   {type=MovieSceneNiagaraSystemTrack, sections [{type=MovieSceneNiagaraSystemSpawnSection, end_frame=4639, start_frame=3732}]}
   {
    type=MovieSceneNiagaraFloatParameterTrack
    display_name=BeamWidth
    sections [{type=MovieSceneFloatSection, channels [None {type=MovieSceneScriptingFloatChannel, default=0, keys [[4577,1],[4588,0]]}]}]
   }
   {
    type=MovieSceneBoolTrack
    display_name=Auto Activate
    sections [{type=MovieSceneBoolSection, channels [None {type=MovieSceneScriptingBoolChannel, default=false, keys [[3715,false],[3732,true]]}]}]
   }
   {
    type=MovieSceneNiagaraFloatParameterTrack
    display_name=ImpactScale
    sections [{type=MovieSceneFloatSection, channels [None {type=MovieSceneScriptingFloatChannel, default=10}]}]
   }
   {
    type=MovieSceneNiagaraFloatParameterTrack
    display_name=ImpactOffset
    sections [{
      type=MovieSceneFloatSection
      channels [None {
        type=MovieSceneScriptingFloatChannel
        default=17600
        thinned_from=62
        keys [[3741,19000],[3745,18000],[3752,17000],[3765,15000],[3780,15000],[3784,16000],[3789,17000],[3804,18000],[3820,18000],[3835,19000],[3842,19000],[3847,20000],[3859,20000],[3863,21000],[3870,21000],[3876,19518.5176],[3900,18000],[3903,18000],[3910,17500],[3918,17500],[3923,16500],[3935,16500],[3940,19000],[3944,19000],[3946,21000],[3956,21000],[3957,20000],[3976,20000],[3985,19000],[3991,19000],[3998,17000],[4004,16000],[4012,14000],[4032,14000],[4046,16000],[4060,17000],[4068,18500],[4080,19500],[4086,20500],[4093,20500],[4105,19500],[4116,19500],[4121,18500],[4125,17500],[4133,16500],[4135,16500],[4151,15500],[4174,15500],[4193,17000],[4312,17000],[4320,19000],[4326,19000],[4328,17000],[4334,17000],[4340,19000],[4356,21000],[4377,20500],[4405,19000],[4495,18200],[4550,17600]]
       }]
     }]
   }
  ]
 }
 PPV_PistolCanary_VFX {
  id=#3
  possessed_class=PostProcessVolume
  tracks [{
    type=MovieSceneFloatTrack
    display_name=Blend Weight
    sections [{type=MovieSceneFloatSection, channels [None {type=MovieSceneScriptingFloatChannel, default=0, keys [[3617,0],[3710,1],[4358,1],[4402,0]]}]}]
   }]
 }
 TentacleBoss {id=#4, possessed_class=PistolCanary_TentacleBoss_C, tracks []}
 SkeletalMesh {
  id=#5
  parent_id=#4
  possessed_class=SkeletalMeshComponent
  tracks [
   {
    type=MovieSceneComponentMaterialTrack
    display_name=Material Slot: Body_A
    sections [{
      type=MovieSceneComponentMaterialParameterSection
      end_frame=4389
      start_frame=3605
      channels [
       PulseSpeed {type=MovieSceneScriptingFloatChannel, default=0, const [0,1]}
       PulseManualOffset {type=MovieSceneScriptingFloatChannel, default=0, keys [[3646.2625,0],[3811.315,1],[3812.1725,0]]}
       GlowPulseEXP {type=MovieSceneScriptingFloatChannel, default=0.079, keys [[3611.1125,0.079],[3632.51,0.069],[3647.7925,0.079]]}
       EmissiveGlowColorR {type=MovieSceneScriptingFloatChannel, const [12,2]}
       EmissiveGlowColorG {type=MovieSceneScriptingFloatChannel, const [2.2855,2]}
       EmissiveGlowColorB {type=MovieSceneScriptingFloatChannel, default=974.9821, keys [[3646.2625,1.8839],[3718.0925,100],[4273,100],[4376,1.8839]]}
       EmissiveGlowColorA {type=MovieSceneScriptingFloatChannel, const [1,2]}
      ]
     }]
   }
   {
    type=MovieSceneComponentMaterialTrack
    display_name=Material Slot: Body_B
    sections [{
      type=MovieSceneComponentMaterialParameterSection
      end_frame=4389
      start_frame=3611
      channels [
       PulseManualOffset {type=MovieSceneScriptingFloatChannel, default=0, keys [[3646.2625,0],[3811.315,1],[3812.1725,0]]}
       PulseSpeed {type=MovieSceneScriptingFloatChannel, default=0, const [0,1]}
       GlowPulseEXP {type=MovieSceneScriptingFloatChannel, default=0.079, keys [[3611.1125,0.079],[3632.51,0.069],[3647.7925,0.079]]}
       EmissiveGlowColorR {type=MovieSceneScriptingFloatChannel, const [12,2]}
       EmissiveGlowColorG {type=MovieSceneScriptingFloatChannel, const [2.2855,2]}
       EmissiveGlowColorB {type=MovieSceneScriptingFloatChannel, default=1.8839, keys [[3655.4325,1.8839],[3719.62,100],[4273,100],[4376,1.8839]]}
       EmissiveGlowColorA {type=MovieSceneScriptingFloatChannel, const [1,2]}
      ]
     }]
   }
   {
    type=MovieSceneComponentMaterialTrack
    display_name=Material Slot: Body_C
    sections [{
      type=MovieSceneComponentMaterialParameterSection
      end_frame=4389
      start_frame=3606
      channels [
       PulseSpeed {type=MovieSceneScriptingFloatChannel, default=0, const [0,1]}
       PulseManualOffset {type=MovieSceneScriptingFloatChannel, default=0, keys [[3646.2625,0],[3811.315,1],[3812.1725,0]]}
       GlowPulseEXP {type=MovieSceneScriptingFloatChannel, keys [[3609.585,0.079],[3630.98,0.069],[3646.2625,0.079],[3984.01,0.079]]}
       EmissiveGlowColorR {type=MovieSceneScriptingFloatChannel, const [12,2]}
       EmissiveGlowColorG {type=MovieSceneScriptingFloatChannel, const [2.2855,2]}
       EmissiveGlowColorB {type=MovieSceneScriptingFloatChannel, default=1.8839, keys [[3664.6025,1.8839],[3721.1475,100],[4273,100],[4376,1.8839]]}
       EmissiveGlowColorA {type=MovieSceneScriptingFloatChannel, const [1,2]}
      ]
     }]
   }
   {
    type=MovieSceneComponentMaterialTrack
    display_name=Material Slot: Body_D
    sections [{
      type=MovieSceneComponentMaterialParameterSection
      end_frame=4389
      start_frame=3608
      channels [
       PulseManualOffset {type=MovieSceneScriptingFloatChannel, default=0, keys [[3646.2625,0],[3811.315,1],[3812.1725,0]]}
       PulseSpeed {type=MovieSceneScriptingFloatChannel, default=0, const [0,2]}
       GlowPulseEXP {type=MovieSceneScriptingFloatChannel, keys [[3609.585,0.079],[3630.98,0.069],[3646.2625,0.079]]}
       EmissiveGlowColorR {type=MovieSceneScriptingFloatChannel, const [12,2]}
       EmissiveGlowColorG {type=MovieSceneScriptingFloatChannel, const [2.2855,2]}
       EmissiveGlowColorB {type=MovieSceneScriptingFloatChannel, default=1.8839, keys [[3672.2425,1.8839],[3721.1475,100],[4273,100],[4376,1.8839]]}
       EmissiveGlowColorA {type=MovieSceneScriptingFloatChannel, const [1,2]}
      ]
     }]
   }
  ]
 }
 CameraComponent {
  id=#6
  parent_id=#7
  possessed_class=CineCameraComponent
  tracks [
   {
    type=MovieSceneFloatTrack
    display_name=Current Focal Length
    sections [{type=MovieSceneFloatSection, channels [None {type=MovieSceneScriptingFloatChannel, default=10.8907}]}]
   }
   {
    type=MovieSceneFloatTrack
    display_name="Manual Focus Distance (Focus Settings)"
    sections [{type=MovieSceneFloatSection, channels [None {type=MovieSceneScriptingFloatChannel, default=100000}]}]
   }
   {
    type=MovieSceneFloatTrack
    display_name=Current Aperture
    sections [{type=MovieSceneFloatSection, channels [None {type=MovieSceneScriptingFloatChannel, default=2.6622}]}]
   }
   {type=MovieSceneCameraShakeTrack, sections [{type=MovieSceneCameraShakeSection, end_frame=3854, start_frame=3815}]}
   {
    type=MovieSceneFloatTrack
    display_name="Intensity (Post Process Settings)"
    sections [{type=MovieSceneFloatSection, channels [None {type=MovieSceneScriptingFloatChannel, default=1.915, keys [[3815,0],[3817,1.915],[3875,0]]}]}]
   }
  ]
 }
 NS_TentacleAttackChargeUpLaser6 {
  id=#8
  possessed_class=NiagaraActor
  tracks [{type=MovieScene3DAttachTrack, sections [{type=MovieScene3DAttachSection, end_frame=4400, start_frame=3595}]}]
 }
 CineCameraActor {
  id=#7
  possessed_class=CineCameraActor
  tracks [
   {
    type=MovieScene3DTransformTrack
    sections [{
      type=MovieScene3DTransformSection
      channels [
       Location.X {type=MovieSceneScriptingDoubleChannel, default=16043, const [16043,1]}
       Location.Y {type=MovieSceneScriptingDoubleChannel, default=50138, const [50138,1]}
       Location.Z {type=MovieSceneScriptingDoubleChannel, default=12608, const [12608,1]}
       Rotation.X {type=MovieSceneScriptingDoubleChannel, default=0}
       Rotation.Y {type=MovieSceneScriptingDoubleChannel, default=-20, const [-20,1]}
       Rotation.Z {type=MovieSceneScriptingDoubleChannel, default=114, const [114,1]}
       Scale.X {type=MovieSceneScriptingDoubleChannel, default=1}
       Scale.Y {type=MovieSceneScriptingDoubleChannel, default=1}
       Scale.Z {type=MovieSceneScriptingDoubleChannel, default=1}
      ]
     }]
   }
   {type=MovieSceneSpawnTrack, sections [{type=MovieSceneSpawnSection, channels [None {type=MovieSceneScriptingBoolChannel, default=true}]}]}
  ]
 }
 NS_TentacleCircularLaser2 {
  id=#2
  possessed_class=NiagaraActor
  tracks [
   {
    type=MovieScene3DTransformTrack
    sections [{
      type=MovieScene3DTransformSection
      channels [
       Location.X {type=MovieSceneScriptingDoubleChannel, default=0}
       Location.Y {type=MovieSceneScriptingDoubleChannel, default=0}
       Location.Z {type=MovieSceneScriptingDoubleChannel, default=0}
       Rotation.X {
        type=MovieSceneScriptingDoubleChannel
        default=0
        thinned_from=12
        keys [[3744,0],[3953,0],[3988,1.7285],[4032,7.9008],[4056,8.6593],[4100,7.9555],[4159,6.8144],[4228,5.9588],[4300,5.9588],[4355,-1.0473]]
       }
       Rotation.Y {
        type=MovieSceneScriptingDoubleChannel
        default=0
        keys [[3744,-12.8215],[3813,5.2743],[3887,19.3979],[3953,11.8179],[3988,-5.1116],[4032,-44.0374],[4056,-48.9793],[4100,-44.4409],[4159,-33.6111],[4228,-17.8474],[4300,-17.8474],[4355,-18.7569]]
       }
       Rotation.Z {
        type=MovieSceneScriptingDoubleChannel
        default=0
        thinned_from=12
        keys [[3744,0],[3953,0],[3988,-18.7118],[4032,-36.926],[4056,-37.9706],[4100,-37.0045],[4159,-35.2005],[4228,-33.2482],[4300,-33.2482],[4355,-11.1873]]
       }
       Scale.X {type=MovieSceneScriptingDoubleChannel, default=1}
       Scale.Y {type=MovieSceneScriptingDoubleChannel, default=1}
       Scale.Z {type=MovieSceneScriptingDoubleChannel, default=1}
      ]
     }]
   }
   {type=MovieScene3DAttachTrack, sections [{type=MovieScene3DAttachSection, end_frame=4709, start_frame=3731}]}
  ]
 }
 NiagaraComponent0 {
  id=#9
  parent_id=#8
  possessed_class=NiagaraComponent
  tracks [
   {type=MovieSceneNiagaraSystemTrack, sections [{type=MovieSceneNiagaraSystemSpawnSection, end_frame=4230, start_frame=3630}]}
   {
    type=MovieSceneNiagaraVectorParameterTrack
    display_name=ChargeScaleVec
    sections [{
      type=MovieSceneFloatVectorSection
      channels [
       Vector.X {type=MovieSceneScriptingFloatChannel, default=0.6, keys [[3632,0],[3637,0.6],[3727,1]]}
       Vector.Y {type=MovieSceneScriptingFloatChannel, default=1, const [1,2]}
       Vector.Z {type=MovieSceneScriptingFloatChannel, default=1, const [1,2]}
      ]
     }]
   }
   {
    type=MovieSceneNiagaraFloatParameterTrack
    display_name=ChargeProgress
    sections [{type=MovieSceneFloatSection, channels [None {type=MovieSceneScriptingFloatChannel, default=1, keys [[3631,0],[3663,0.3],[3722,1]]}]}]
   }
   {
    type=MovieSceneNiagaraFloatParameterTrack
    display_name=PlasmaEnabled
    sections [{type=MovieSceneFloatSection, channels [None {type=MovieSceneScriptingFloatChannel, default=0, keys [[3728,1],[3737,0]]}]}]
   }
   {
    type=MovieSceneNiagaraColorParameterTrack
    display_name=ChargeScaleColor
    sections [{
      type=MovieSceneColorSection
      channels [
       Color.R {type=MovieSceneScriptingFloatChannel, default=1}
       Color.G {type=MovieSceneScriptingFloatChannel, default=1}
       Color.B {type=MovieSceneScriptingFloatChannel, default=1}
       Color.A {type=MovieSceneScriptingFloatChannel, default=0, keys [[3646,0],[3681,1],[4006,1],[4106,0]]}
      ]
     }]
   }
   {
    type=MovieSceneNiagaraFloatParameterTrack
    display_name=ChargeProgressLightning
    sections [{type=MovieSceneFloatSection, channels [None {type=MovieSceneScriptingFloatChannel, default=0.5472, keys [[3633,0],[3663,0.5472],[3686,1]]}]}]
   }
   {
    type=MovieSceneNiagaraFloatParameterTrack
    display_name=PostProcess
    sections [{type=MovieSceneFloatSection, channels [None {type=MovieSceneScriptingFloatChannel, default=1, keys [[3634,0],[3742,1],[4216,0]]}]}]
   }
   {
    type=MovieSceneNiagaraFloatParameterTrack
    display_name=TipDelay
    sections [{type=MovieSceneFloatSection, channels [None {type=MovieSceneScriptingFloatChannel, default=0.4}]}]
   }
   {
    type=MovieSceneParticleTrack
    display_name=FX System
    sections [{
      type=MovieSceneParticleSection
      channels [None {type=MovieSceneScriptingParticleChannel, keys [[3630, "<ParticleKey.ACTIVATE: 0>"], [4250, "<ParticleKey.DEACTIVATE: 1>"]]}]
     }]
   }
   {
    type=MovieSceneNiagaraFloatParameterTrack
    display_name=LightningEnabled
    sections [{type=MovieSceneFloatSection, channels [None {type=MovieSceneScriptingFloatChannel, default=0, keys [[3736,1],[3743,0]]}]}]
   }
  ]
 }
]
display_rate {denominator=1, numerator=60}
tick_resolution {denominator=1, numerator=24000}
tracks []
